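# M_Laser_ImpactBurn is an unlit, additive, two-sided particle material, named and located as a Space Pistol laser impact burn effect.
# It multiplies the RGB of the T_HitImpactFlare texture by particle color, then by particle alpha, and feeds the result into the MF_FX_EmissiveAdjustments material function.
# It is flagged for use on both Niagara and particle sprites.
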Material /Game/Weapons/FORT_Special/SpacePistol/Materials/M_Laser_ImpactBurn
graph {
 properties {
  bUsedWithNiagaraSprites=true
  bUsedWithParticleSprites=true
  blendMode=BLEND_Additive
  pixelDepthOffsetMode=PDOM_Legacy
  shadingModel=MSM_Unlit
  twoSided=true
  usedShadingModels=Unlit
 }
 nodes [
  n1 = ParticleColor()
  n2 = TextureSample() {texture=/Game/Effects/Fort_Effects/Textures/T_HitImpactFlare}
  MaterialFunctionCall(Emissive=Multiply(A=Multiply(A=@n1.RGB, B=@n2.RGB), B=@n1.A)) {materialFunction=&MF_FX_EmissiveAdjustments}
 ]
}
