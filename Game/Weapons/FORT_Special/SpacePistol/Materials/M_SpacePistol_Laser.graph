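# Unlit material for the SpacePistol's laser bolt.
# A vector parameter, Bolt Color, defaults to a bright HDR blue and feeds the Emissive input of the MF_FX_EmissiveAdjustments function.
# It is flagged for use on mesh particles in both Cascade and Niagara.
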Material /Game/Weapons/FORT_Special/SpacePistol/Materials/M_SpacePistol_Laser
graph {
 properties {
  bUsedWithMeshParticles=true
  bUsedWithNiagaraMeshParticles=true
  pixelDepthOffsetMode=PDOM_Legacy
  shadingModel=MSM_Unlit
  usedShadingModels=Unlit
 }
 nodes [MaterialFunctionCall(Emissive=VectorParameter() {parameterName=Bolt Color, defaultValue=<rgba 2.1254 4.7024 10 0>}.RGB) {materialFunction=&MF_FX_EmissiveAdjustments}]
}
